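# List entry widget for the event binding picker in the Creative property editor, derived from FortCreativePropertyEditorEventBindingDialogListEntry.
# A fixed-height row overlays a content widget on a background widget, and passes the display name to the content widget.
# Hover, selection and list-item assignment events call UpdateVisuals.
# UpdateVisuals treats selected and hovered as one highlight flag, switching the content state and the background's type and focus style.
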WidgetBlueprint /Game/Athena/UI/InGame/Creative/PropertyEditor/DirectBinding/CreativePropertyEventBindingPickerListEntry
graphs [
 EventGraph {nodes [
   n1 = Event OnDisplayNameSet!() -> (OutputDelegate, then, DisplayName:"Text (by ref)") {is=title K2Node_Event}
   > UpdateText!(
    execute
    self:WBP Creative Property Editor Content Primary Object Reference=Get WBP_CreativePropertyEditor_Content_Primary() -> (WBP_CreativePropertyEditor_Content_Primary:WBP Creative Property Editor Content Primary Object Reference)
    Text:"Text (by ref)"=@n1.DisplayName
   ) -> (then)
   n2 = Event On Selected()
   UpdateVisuals(execute=[@n2.then, @n3.then, @n4.then, @n5.then, @n6.then])
   n3 = Event On Hovered()
   n4 = Event On Deselected()
   n5 = Event On Unhovered()
   n6 = Event OnListItemObjectSet()
  ]}
 UpdateVisuals {nodes [
   UpdateVisuals() {is=title K2Node_FunctionEntry}
   > n1 = Set bLocal_IsSelectedOrHovered(bLocal_IsSelectedOrHovered:bool=OR Boolean(A=GetSelected!(self:Common Button Base Object Reference) -> (ReturnValue:bool), B=IsHovered())) -> (Output_Get:bool)
   > Event Update States!(
    execute
    self:WBP Creative Property Editor Content Primary Object Reference=Get WBP_CreativePropertyEditor_Content_Primary()
    State:E_UI_ChannelsBindingStates Enum=Select(Option 0:E_UI_ChannelsBindingStates Enum=NewEnumerator7, Option 1:E_UI_ChannelsBindingStates Enum=NewEnumerator8, Index=@n1) -> (ReturnValue:E_UI_ChannelsBindingStates Enum)
   ) -> (then)
   n2 = Get bLocal_IsSelectedOrHovered() -> (bLocal_IsSelectedOrHovered:bool)
   > Event Update Background Style!(
    execute
    self:WBP Creative Property Editor Background Skewed With Marker Object Reference=Get WBP_CreativePropertyEditor_Background_SkewWithMarker() -> (WBP_CreativePropertyEditor_Background_SkewWithMarker:WBP Creative Property Editor Background Skewed With Marker Object Reference)
    InType:E_UI_CreativePropertyEditor_Type Enum=Select(Option 0:E_UI_CreativePropertyEditor_Type Enum=NewEnumerator4, Option 1:E_UI_CreativePropertyEditor_Type Enum=NewEnumerator3, Index=@n2) -> (ReturnValue:E_UI_CreativePropertyEditor_Type Enum)
    InMarkerState:E_UI_CreativePropertyEditor_MarkerStates Enum=NewEnumerator0
    InFocusState:E_UI_CreativePropertyEditor_FocusStates Enum=Select(
     Option 0:E_UI_CreativePropertyEditor_FocusStates Enum=NewEnumerator0
     Option 1:E_UI_CreativePropertyEditor_FocusStates Enum=NewEnumerator1
     Index=@n2
    ) -> (ReturnValue:E_UI_CreativePropertyEditor_FocusStates Enum)
   ) -> (then)
  ]}
]
variables []
widgets {
 info {
  inheritedWidgetCount=2
  parentClass=/Script/FortniteUI.FortCreativePropertyEditorEventBindingDialogListEntry
  rootWidgetClass=&SizeBox
  widgetCount=6
 }
 widgets [
  SizeBox_0 {widgetClassPath=&SizeBox, widget=SizeBox_0, properties {bOverride_HeightOverride=true, heightOverride=50}}
  Overlay_12 {widgetClassPath=&Overlay, parent=SizeBox_0, widget=Overlay_12, slot {class=SizeBoxSlot, properties {padding=<ltrb 11 0 11 0>}}}
  WBP_CreativePropertyEditor_Background_SkewWithMarker {
   widgetClassPath=/Game/Athena/UI/InGame/Creative/PropertyEditor/WBP_CreativePropertyEditor_Background_SkewedWithMarker
   bIsVariable=true
   parent=Overlay_12
   widget=WBP_CreativePropertyEditor_Background_SkewWithMarker
   properties {displayLeftSkew=false, displayMarker=false, displayRightSkew=false, leftSkewEdgeWidth=10, rightSkewEdgeWidth=10}
   slot {class=OverlaySlot, properties {horizontalAlignment=HAlign_Fill, verticalAlignment=VAlign_Fill}}
  }
  WBP_CreativePropertyEditor_Content_Primary {
   widgetClassPath=/Game/Athena/UI/InGame/Creative/PropertyEditor/WBP_CreativePropertyEditor_Content_Primary
   bIsVariable=true
   parent=Overlay_12
   widget=WBP_CreativePropertyEditor_Content_Primary
   properties {states=Default}
   slot {class=OverlaySlot, properties {horizontalAlignment=HAlign_Fill, verticalAlignment=VAlign_Center, padding=<ltrb 5 0 5 0>}}
  }
  Text_EntryName {widgetClassPath=&CommonTextBlock, bInherited=true, widget=null}
  InputActionWidget {widgetClassPath=&CommonActionWidget, bInherited=true, widget=null}
 ]
}
